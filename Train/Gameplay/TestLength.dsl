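; TestLength is a debug Blueprint in the train gameplay folder that reports the length of a track's spline.
; On BeginPlay, only the authority starts a looping one-second timer bound to a custom event.
; Each tick of the timer finds the SplineComponent on the Track actor variable, converts its length to whole meters and prints it on screen and to the log.
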(event EventBeginPlay
  (Networking|SwitchHasAuthority
    (:Authority
      (Utilities|Time|SetTimerbyEvent (delegate Custom|CustomEvent) 1.0 true))))

(event Custom|CustomEvent
  (Development|PrintString (Utilities|String|Append "Track Length = " (Utilities|String|ToString(Integer) (Math|Float|Truncate (Math|Float|SafeDivide (Spline|GetSplineLength (Actor|GetComponentbyClass (Variables|Default|GetTrack) "/Script/Engine.SplineComponent")) 1e+02))) " Meters") true true "(R=0.000000,G=0.660000,B=1.000000,A=1.000000)" 1.0))

(fn ConstructionScript)
